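# Master masked road material for Hera terrain spline meshes, with Nanite support.
# Asphalt with painted lines, sand edges and an optional Custom01 layer combine through height blends, UV fade-outs and road-type falloff masks, then optional snow.
# It writes to a runtime virtual texture and switchably reads the landscape's, reapplying snow color after the read.
# Mobile uses simpler asphalt, plus opacity masking and distance-scaled offset toward the camera, replaced by a downward offset for shadows.
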Material /Hera/Environment/Terrain/Roads/Material/M_Hera_Roads_Base
graph {
 properties {
  bAutomaticallySetUsageInEditor=false
  bCastRayTracedShadows=false
  bTangentSpaceNormal=false
  bUseMaterialAttributes=true
  bUsedWithNanite=true
  bUsedWithSplineMeshes=true
  blendMode=BLEND_Masked
  physMaterial=/Game/FortressPhysicalMaterials/Asphalt
  pixelDepthOffsetMode=PDOM_Legacy
  usedShadingModels=Default Lit
 }
 nodes [
  Constant4Vector() {constant=<rgba 0.2747 0.3916 0.4793 1>}
  Constant() {r=1.4}
  Constant() {r=2}
  Constant() {r=0.5}
  Constant() {r=0.8}
  n1 = NamedRerouteUsage()
  n2 = NamedRerouteUsage()
  n3 = LinearInterpolate(
   A=MaterialFunctionCall(BaseNormal=NamedRerouteUsage(), AdditionalNormal=@n1) {materialFunction=&BlendAngleCorrectedNormals}.Result
   B=@n1
   Alpha=@n2
  )
  LinearInterpolate(A=@n3, B=NamedRerouteUsage(), Alpha=NamedRerouteUsage())
  n4 = NamedRerouteUsage() {desc=May need this version with the small snow detail breakup}
  n5 = SphereMask(
   A=NamedRerouteUsage()
   B=Constant() {r=1}
   Radius=ScalarParameter() {defaultValue=2, group=RoadSnow, parameterName=RoadSnow_RoadRadius}
   Hardness=Constant()
  )
  n6 = ScalarParameter() {defaultValue=1.5, group=RoadSnow, parameterName=RoadSnow_GradientStrength}
  n7 = Saturate(None=Subtract(A=Multiply(A=@n5, B=@n6), B=NamedRerouteUsage()))
  n8 = LinearInterpolate(A=Multiply(A=@n7, B=NamedRerouteUsage()), B=@n7, Alpha=NamedRerouteUsage())
  n9 = Saturate(None=Add(A=OneMinus(None=@n4), B=NamedRerouteDeclaration(None=@n8) {name=SnowRoadInnerFalloffMask, nodeColor=<rgba 1 0.025 0.3669 1>}))
  n10 = NamedRerouteUsage()
  LinearInterpolate(A=Add(A=@n9, B=OneMinus(None=@n10)), B=@n9, Alpha=NamedRerouteUsage())
  n11 = Constant() {r=1}
  n12 = GetMaterialAttributes(None=NamedRerouteUsage()) {attributeGetTypes [#1]}
  n13 = NamedRerouteUsage()
  n14 = LinearInterpolate(A=@n12.Base Color, B=NamedRerouteUsage(), Alpha=@n13)
  n15 = Multiply(A=NamedRerouteUsage(), B=@n13)
  n16 = NamedRerouteDeclaration(None=StaticSwitchParameter(
    True=NamedRerouteUsage() {desc=Would probably be better to have the version that's getting written to RVT to use a neutral value for the snow basecolor?}
    False=NamedRerouteUsage()
   ) {group=RoadSnow, parameterName=UseRoadSnow, sortPriority=-5000}) {name=FinalMaterialWithNoRVT, nodeColor=<rgba 0.5 0.5 0 1>}
  n17 = StaticSwitchParameter(
   True=StaticSwitchParameter(
    True=SetMaterialAttributes!(MaterialAttributes=@n12.MaterialAttributes, Base Color=@n14, Emissive Color=@n15) {attributeSetTypes [#1, #2]}
    False=NamedRerouteUsage()
   ) {group=RoadSnow, parameterName=UseRoadSnow, sortPriority=-5000}
   False=@n16
  ) {
   defaultValue=true
   desc="This should be set to false for roads on things like bridges, where RVT blending is not desired"
   group=RVT
   parameterName=ReadRVT
  }
  n18 = MaterialFunctionCall(
   NoRVT=@n16
   RVT=MaterialFunctionCall!(RVT=@n17, ByPass RVT=@n16) {materialFunction=/Game/Athena/Apollo/Maps/Landscape/VT/MaterialFunctions/EnableRVT}.Result
  ) {materialFunction=&FN_RVT_Switch_Material}
  n19 = ScalarParameter() {
   defaultValue=1
   desc="A temp fix for Mobile rendering bug where static 1.0 in opacity + WPO is causing it to render black.  Serge Bernier is looking into a permanent fix and then we can swap this back to a static 1.0 if desired."
   group=ZDoNotSet
   parameterName=SwitchMaskOpacity
   sliderMax=1
   sliderMin=1
  }
  n20 = WorldPosition()
  n21 = CameraPositionWS()
  n22 = Divide(A=Distance(A=@n20.XYZ, B=@n21)) {constB=700}
  n23 = Normalize(VectorInput=Subtract(A=@n21, B=@n20.XYZ))
  n24 = SetMaterialAttributes!(
   MaterialAttributes=@n18.Result
   Opacity Mask=QualitySwitch(Default=NamedRerouteUsage(), Medium=@n19)
   World Position Offset=ShadowReplace(Default=Multiply(A=@n22, B=@n23), Shadow=Constant3Vector() {constant=<rgba 0 0 -60 0>})
  ) {
   desc="The opacity mask is used on mobile - this material is using Masked blend mode!"
   attributeSetTypes [#3, #4]
  }
  MaterialProxyReplace(
   Realtime=ShadingPathSwitch(
    Default=SetMaterialAttributes!(MaterialAttributes=@n18.Result, Opacity Mask=@n11) {desc="Masked blend mode is enabled, but only mobile utilises it!", attributeSetTypes [#3]}
    Mobile=@n24
   )
   MaterialProxy=@n24
  )
  n25 = Constant() {r=5}
  n26 = Constant() {r=-1.25}
  n27 = NamedRerouteUsage()
  n28 = Saturate(None=Add(A=NamedRerouteUsage(), B=Constant() {r=0.4}))
  n29 = VertexColor()
  n30 = Multiply(A=OneMinus(None=@n27), B=Constant() {r=1.25}) {constB=3}
  n31 = Multiply(
   A=Saturate(None=LinearInterpolate(A=@n25, B=@n26, Alpha=@n27))
   B=MaterialFunctionCall(Transition Phase=Multiply(A=@n28, B=@n29.R), Height Texture=Saturate(None=@n30), Contrast=Constant() {r=0.5}) {materialFunction=&HeightLerp}.Alpha
  )
  Multiply(A=@n31, B=VertexColor().R)
  n32 = VectorParameter() {parameterName=DEBUG_OverrideColor, channelNames {a=ColorAmount, b="", g="", r=""}, defaultValue=<rgba 1 0 0 1>}
  n33 = GetMaterialAttributes(None=@n18.Result) {attributeGetTypes [#1]}
  Multiply(A=@n32, B=@n33.Base Color)
  n34 = GetMaterialAttributes(None=NamedRerouteUsage()) {attributeGetTypes [
    #5
    #6
    #1
    #7
    #3
   ]}
  n35 = ScalarParameter() {defaultValue=1, group=FadeOut, parameterName=FadeOut_HeightMapAmount}
  NamedRerouteDeclaration(None=Saturate(None=Multiply(A=@n34.Opacity Mask, B=@n35))) {name=High_AsphaltHeight, nodeColor=<rgba 0.5469 0 1 1>}
  n36 = MaterialFunctionCall!(Scale=Constant() {r=600}) {materialFunction=/Game/Athena/Apollo/Environments/Landscape/Materials/Biome/CommonMaterials/MF_SimpleWorldUV}
  n37 = TextureSample(UVs=@n36.Result) {samplerSource=SSM_Wrap_WorldGroupSettings, texture=/Game/Athena/Apollo/Environments/Landscape/Textures/Roads/T_AsphaltDetail1_D}
  n38 = Constant() {r=0.7}
  n39 = TextureSample(UVs=@n36.Result) {
   samplerSource=SSM_Wrap_WorldGroupSettings
   samplerType=SAMPLERTYPE_Masks
   texture=/Game/Athena/Apollo/Environments/Landscape/Textures/Roads/T_AsphaltDetail1_S
  }
  n40 = Constant() {r=0.2}
  n41 = TextureSample(UVs=@n36.Result) {
   samplerSource=SSM_Wrap_WorldGroupSettings
   samplerType=SAMPLERTYPE_Normal
   texture=/Game/Athena/Apollo/Environments/Landscape/Textures/Roads/T_AsphaltDetail1_N
  }
  n42 = SetMaterialAttributes!(
   MaterialAttributes
   Base Color=Saturate(None=Divide(A=@n37.RGB, B=@n38))
   Specular=Multiply(A=Constant() {r=0.5}, B=@n39.R)
   Roughness=Saturate(None=Add(A=@n40, B=@n39.B))
   Normal=MaterialFunctionCall!(Normals=@n41.RGB, TransformToTangentSpace, Normal Strength=Constant() {r=4}) {materialFunction=/Game/Athena/Apollo/Environments/Landscape/Materials/Biome/CommonMaterials/MF_HQ_WorldAlignedNormal}.Result
   Opacity Mask=NamedRerouteDeclaration(None=@n39.B) {name=AsphaltSidesHeight, nodeColor=<rgba 0.5469 0 1 1>}
  ) {attributeSetTypes [
    #1
    #6
    #5
    #7
    #3
   ]}
  NamedRerouteDeclaration(None=@n42) {name=AsphaltSidesMaterial, nodeColor=<rgba 0.5469 0 1 1>}
  n43 = NamedRerouteUsage()
  n44 = Max(A=NamedRerouteUsage(), B=Power(Base=NamedRerouteUsage(), Exp=Constant() {r=4}))
  n45 = NamedRerouteUsage()
  n46 = NamedRerouteUsage()
  n47 = MaterialFunctionCall(Transition Phase=Multiply(A=LinearInterpolate(A=@n43, B=@n44, Alpha=@n45), B=@n46), Height Texture=@n46) {materialFunction=&HeightLerp}
  n48 = NamedRerouteUsage()
  n49 = NamedRerouteUsage()
  n50 = NamedRerouteUsage()
  n51 = Multiply(
   A=MaterialFunctionCall(
    Transition Phase=MaterialFunctionCall(Transition Phase=@n48, Height Texture=@n49) {materialFunction=&HeightLerp}.Lerp Alpha No Contrast
    Height Texture=Multiply(A=@n50, B=@n50)
   ) {materialFunction=&HeightLerp}.Lerp Alpha No Contrast
   B=NamedRerouteUsage()
  )
  n52 = NamedRerouteUsage()
  n53 = NamedRerouteUsage()
  n54 = LinearInterpolate(B=NamedRerouteUsage(), Alpha=ScalarParameter() {group=Asphalt_SandEdge, parameterName=AsphaltSand_UseFadeOutMask}) {constA=1}
  n55 = NamedRerouteUsage()
  n56 = Max(A=StaticSwitch(True=LinearInterpolate(A=@n52, B=@n53, Alpha=@n54), False=NamedRerouteUsage(), Value=NamedRerouteUsage()), B=@n55)
  n57 = QualitySwitch(Default=@n55, Medium=@n56)
  NamedRerouteDeclaration(None=StaticSwitch(
    True=ShadingPathSwitch(Default=@n47.Lerp Alpha No Contrast, Mobile=@n51)
    False=ShadingPathSwitch(Default=@n56, Mobile=@n57)
    Value=NamedRerouteUsage()
   )) {name=AsphaltRoadOpacityMask, nodeColor=<rgba 0.5469 0 1 1>}
  n58 = ComponentMask(None=TextureCoordinate()) {g=true}
  n59 = OneMinus(None=@n58)
  n60 = ScalarParameter() {group=FadeOut, parameterName=InvertFadeDirection_Scalar, sortPriority=-500}
  NamedRerouteDeclaration(None=Saturate(None=LinearInterpolate(A=@n58, B=@n59, Alpha=@n60))) {name=FadeOutMask, nodeColor=<rgba 0.5469 0 1 1>}
  n61 = TextureSampleParameter2D() {
   group=Road Lines
   parameterName=Road Line Normal
   samplerSource=SSM_Wrap_WorldGroupSettings
   samplerType=SAMPLERTYPE_Normal
   texture=/Game/Athena/Apollo/Environments/Landscape/Textures/Roads/AsphaltRoad1_normal
  }
  n62 = Constant3Vector() {constant=<rgba 1.7 1.7 1 0>}
  NamedRerouteDeclaration(None=Transform(None=Multiply(A=@n61.RGB, B=@n62))) {name=RoadLineNormal, nodeColor=<rgba 0.5469 0 1 1>}
  n63 = NamedRerouteUsage()
  n64 = VectorParameter() {group=RoadSnow, parameterName=RoadSnow_BaseColorReplace_Mobile, defaultValue=<rgba 0.2546 0.3222 0.38 1>}
  n65 = ShadingPathSwitch(
   Default=VectorParameter() {group=RoadSnow, parameterName=RoadSnow_BaseColorReplace, defaultValue=<rgba 0.2747 0.3916 0.4793 1>}.RGB
   Mobile=@n64.RGB
  )
  n66 = GetMaterialAttributes(None=MaterialProxyReplace(
    Realtime=SetMaterialAttributes!(MaterialAttributes=@n63, Base Color=@n65) {attributeSetTypes [#1]}
    MaterialProxy=SetMaterialAttributes(MaterialAttributes=@n63, Base Color=@n65) {attributeSetTypes [#1]}
   )) {attributeGetTypes [#3]}
  NamedRerouteDeclaration(None=@n66.MaterialAttributes) {name=Snow_Material, nodeColor=<rgba 1 0.025 0.3669 1>}
  n67 = Multiply(A=TruncateLWC!(None=WorldPosition().XY)) {constB=0.0001}
  n68 = GetMaterialAttributes(None=MaterialFunctionCall!(
    LayerWeight=Constant() {r=1}
    VariationMasks Mobile=TextureSample(UVs=@n67) {samplerSource=SSM_Wrap_WorldGroupSettings, texture=/Game/Environments/Helios/Terrain/TexturesMobile/T_Hermes_Mobile_Noise}.RGBA
   ) {materialFunction=/Game/Athena/Helios/Environments/Landscape/Materials/QualityLevelLayers/MF_Helios_Asphalt_NS}.Result) {attributeGetTypes [#1, #7, #3]}
  n69 = Add(A=Multiply(A=@n68.Normal) {constB=2}) {constB=-1}
  n70 = Abs(None=@n69)
  n71 = Subtract(A=NamedRerouteUsage(), B=Max(A=ComponentMask(None=@n70) {r=true}, B=ComponentMask(None=@n70) {g=true}))
  n72 = Subtract(A=Multiply(A=ComponentMask(None=@n68.Normal) {r=true}) {constB=4}) {constB=1.75}
  n73 = LinearInterpolate(A=@n68.Base Color, B=NamedRerouteUsage(), Alpha=Multiply(A=Saturate(None=@n71), B=Saturate(None=@n72)) {constB=0.6})
  n74 = NamedRerouteUsage()
  n75 = DeriveNormalZ(InXY=@n69)
  n76 = Transform(None=Multiply(A=NamedRerouteUsage(), B=Constant3Vector() {constant=<rgba 1.7 1.7 1 0>}))
  n77 = Add(A=ComponentMask(None=@n75) {g=true, r=true}, B=ComponentMask(None=@n76) {g=true, r=true})
  n78 = Multiply(A=ComponentMask(None=@n75) {b=true}, B=ComponentMask(None=@n76) {b=true})
  NamedRerouteDeclaration(None=SetMaterialAttributes!(MaterialAttributes=@n68.MaterialAttributes, Base Color=Multiply(A=@n73, B=@n74), Normal=AppendVector(A=@n77, B=@n78)) {attributeSetTypes [#1, #7]}) {name=AsphaltWithRoadLines_Mobile, nodeColor=<rgba 0.5469 0 1 1>}
  n79 = MaterialFunctionCall(TextureObject=TextureObject() {texture=/Game/Environments/Sets/Arid/Textures/T_Swirly_Noise_D}, TextureSize=Constant() {r=350}) {materialFunction=&WorldAlignedTexture}
  NamedRerouteDeclaration(None=LinearInterpolate(A=Constant() {r=0.5}, B=Constant() {r=0.9575}, Alpha=ComponentMask(None=@n79.Z Texture) {b=true})) {name=AsphaltRoughnessNoise, nodeColor=<rgba 0.5469 0 1 1>}
  n80 = VectorParameter() {
   group=RoadTypeEdgeFalloff
   parameterName=RoadTypeEdgeFalloff_Large
   sortPriority=-300
   channelNames {a=Add, b=Multiply, g=FalloffMax, r=FalloffMin}
   defaultValue=<rgba -0.6 0.75 1 0>
  }
  n81 = VectorParameter() {
   desc="Default is setup as R=Straight, G=JunctionT, B=JunctionY, A=JunctionSplit"
   group=RoadTypeEdgeFalloff
   parameterName=RoadTypeEdgeFalloff_Channel
   sortPriority=-400
   defaultValue=<rgba 1 0 0 0>
  }
  n82 = MaterialFunctionCall(X=@n81.R, Y=@n81.R, Z=@n81.R, A=@n81.R) {materialFunction=&MakeFloat4}
  n83 = TextureSampleParameter2D(UVs=TextureCoordinate()) {
   desc=CLAMPED
   group=RoadTypeEdgeFalloff
   parameterName=RoadTypeEdgeFalloffMask
   sortPriority=-500
   texture=/Game/Athena/Apollo/Environments/Landscape/Textures/Roads/Helios/T_RoadFalloffMasks
  }
  n84 = MaterialFunctionCall(X=@n83.R, Y=@n83.R, Z=@n83.R, A=Multiply(A=@n83.A, B=@n83.A) {desc=Should this just be done in the texture itself?}) {materialFunction=&MakeFloat4}
  n85 = NamedRerouteDeclaration(None=Saturate(None=DotProduct(A=@n82.Result, B=@n84.Result))) {name=RoadTypeEdgeFalloffMask_Channel, nodeColor=<rgba 0.5469 0 1 1>}
  n86 = Add(A=Multiply(A=SmoothStep(Min=@n80.FalloffMin, Max=@n80.FalloffMin, Value=@n85) {constMax=0.5}, B=@n80.Multiply), B=@n80.Add)
  n87 = VectorParameter() {
   group=RoadTypeEdgeFalloff
   parameterName=RoadType_EdgeFalloff_Small
   sortPriority=-200
   channelNames {a=Add, b=Multiply, g=FalloffMax, r=FalloffMin}
   defaultValue=<rgba -0.015 0.01 1 0>
  }
  n88 = Add(A=Multiply(A=SmoothStep(Min=@n87.FalloffMin, Max=@n87.FalloffMin, Value=@n85) {constMax=0.5}, B=@n87.Multiply), B=@n87.Add)
  NamedRerouteDeclaration(None=Multiply(A=Saturate(None=@n86), B=Saturate(None=@n88))) {name=RoadType_EdgeGradient, nodeColor=<rgba 0.5469 0 1 1>}
  NamedRerouteDeclaration(None=Saturate(None=@n66.Opacity Mask)) {name=Snow_Height, nodeColor=<rgba 1 0.025 0.3669 1>}
  n89 = RuntimeVirtualTextureSample!(UVs, Translated World Position) {
   materialType=BaseColor_Normal_Specular_YCoCg
   virtualTexture=/Game/Athena/Apollo/Maps/Landscape/VT/RVT_Apollo_Landscape
   worldPositionOriginType=CameraRelative
  }
  n90 = SetMaterialAttributes!(MaterialAttributes, Base Color=@n89.BaseColor, Specular=@n89.BaseColor, Roughness=@n89.BaseColor, Normal=@n89.BaseColor) {attributeSetTypes [
    #1
    #6
    #5
    #7
   ]}
  NamedRerouteDeclaration(None=@n90) {name=RVT_READ_Material, nodeColor=<rgba 0.5 0.5 0 1>}
  n91 = MaterialFunctionCall!(Grass_UseCustomSlopeMask, Grass_CustomSlopeMask) {materialFunction=/Hera/Environment/Terrain/Landscape/Materials/Layers/MF_Hera_StaticMeshLandscapeLayerMatch}
  NamedRerouteDeclaration(None=@n91.SnowLayer) {name=LandscapeLayerMatch_Snow, nodeColor=<rgba 0.5469 0 1 1>}
  n92 = VectorParameter() {
   group=RoadSnow
   parameterName=RoadSnow_CoverageNoiseSettings
   channelNames {a=RoadSnow_Noise_Strength, b=RoadSnow_Noise_Max, g=RoadSnow_Noise_Min, r=RoadSnow_Noise_Scale}
   defaultValue=<rgba 4096 -0.2 1 1>
  }
  n93 = TextureSampleParameter2D(UVs=Divide(A=WorldPosition().XY, B=@n92.RoadSnow_Noise_Scale)) {
   group=RoadSnow
   parameterName=RoadSnow_CoverageNoise
   samplerSource=SSM_Wrap_WorldGroupSettings
   texture=/Game/Effects/Fort_Effects/Textures/Tile/T_GaussianNoise_01_M
  }
  NamedRerouteDeclaration(None=Multiply(A=LinearInterpolate(A=@n92.RoadSnow_Noise_Min, B=@n92.RoadSnow_Noise_Min, Alpha=@n93.R), B=@n92.RoadSnow_Noise_Strength)) {name=SnowNoiseMask, nodeColor=<rgba 1 0.025 0.3669 1>}
  NamedRerouteDeclaration(None=@n7) {name=DEBUG_Thing, nodeColor=<rgba 1 0 0 1>}
  n94 = MaterialFunctionCall!(
   ScalarWeightMask
   UseSnowOcclusionTextureMask
   ------
   UseMaterialInput
   AdditionalMaskToRemoveSnow
   OutputTangentNormals
   UseAdditionalMaskToRemoveSnow
   UseSnowNormals
   OuputVertexNormalsAsTheSnowNormals
   UseSparkle
   LQSpark
   AllowSnowMicroNormals
   CliffMaterialOnly
   Material=NamedRerouteUsage()
   CliffLandscapeLayerMask
   IsLandscape=StaticBool()
   AreInputNormalsTangentSpace=StaticBool()
   UseInputNormalsForFakeSnowShadow=StaticBool() {value=true}
  ) {materialFunction=/Hera/Environment/Terrain/Landscape/Materials/Layers/QualityLevel/MF_Hera_Snow_Default}
  NamedRerouteDeclaration(None=Multiply(A=@n94.DetailMask, B=@n94.DetailMask)) {name=SnowBiomeMask, nodeColor=<rgba 1 0.025 0.3669 1>}
  n95 = Constant() {r=-0.5}
  n96 = Constant() {r=4.5}
  n97 = TextureSampleParameter2D!(UVs, Bias=ShadingPathSwitch(Default=Constant() {r=-0.5}, Mobile=Constant() {r=-0.7}), Apply View MipBias) {
   group=Road Lines
   mipValueMode=TMVM_MipBias
   parameterName=Road Line Mask
   samplerSource=SSM_Wrap_WorldGroupSettings
   samplerType=SAMPLERTYPE_Masks
   texture=/Game/Athena/Apollo/Environments/Landscape/Textures/Roads/AsphaltRoad1_rgb
  }
  n98 = NamedRerouteDeclaration(None=@n97.R) {name=RoadLineMask_R, nodeColor=<rgba 0.5469 0 1 1>}
  NamedRerouteDeclaration(None=Saturate(None=LinearInterpolate(A=@n95, B=@n96, Alpha=@n98))) {name=AsphaltOnlyMask, nodeColor=<rgba 0.5469 0 1 1>}
  n99 = ScalarParameter() {defaultValue=0.5, group=RoadSnow, parameterName=RoadSnow_AsphaltHeightA}
  n100 = ScalarParameter() {defaultValue=0.8, group=RoadSnow, parameterName=RoadSnow_AsphaltHeightB}
  n101 = Power(Base=NamedRerouteUsage(), Exp=ScalarParameter() {defaultValue=4, group=RoadSnow, parameterName=RoadSnow_AsphaltHeightPow})
  n102 = MaterialFunctionCall!(
   Material 1=NamedRerouteUsage()
   Height Texture 1=NamedRerouteUsage()
   Material 2=NamedRerouteUsage()
   Height Texture 2=Saturate(None=LinearInterpolate(A=@n99, B=@n100, Alpha=@n101))
   Transistion Phase=@n9
   UseContrast
   Contrast
   AddLayerMask
  ) {materialFunction=/Game/Athena/Helios/Environments/Landscape/Materials/MaterialFunctions_Utilities/MF_HeightBlendLandscapeLayers}
  NamedRerouteDeclaration(None=@n102.BlendedMaterial) {name=Snow_WithNoRVT-Material, nodeColor=<rgba 1 0.025 0.3669 1>}
  n103 = GetMaterialAttributes(None=NamedRerouteUsage()) {attributeGetTypes [#1, #2, #7]}
  NamedRerouteDeclaration(None=@n103.Base Color) {name=Snow_ForPostRVT_Basecolor, nodeColor=<rgba 1 0.025 0.3669 1>}
  NamedRerouteDeclaration(None=@n103.Emissive Color) {name=Snow_ForPostRVT_Emissive, nodeColor=<rgba 1 0.025 0.3669 1>}
  n104 = NamedRerouteUsage()
  n105 = LinearInterpolate(
   A=ScalarParameter() {defaultValue=1, group=Custom01, parameterName=Custom01_Noise_Min_ModColor}
   B=ScalarParameter() {defaultValue=0.4, group=Custom01, parameterName=Custom01_Noise_MaxModColor}
   Alpha=@n104
  )
  n106 = TextureCoordinate()
  n107 = MaterialFunctionCall(Scale=ScalarParameter() {defaultValue=650, group=Custom01, parameterName=Custom01_Scale}) {materialFunction=&MF_SimpleWorldUV}
  n108 = ScalarParameter() {defaultValue=1, desc=False uses UV0, group=Custom01, parameterName=Custom01_UseWorldUV}
  n109 = LinearInterpolate(A=@n106, B=@n107.Result, Alpha=@n108)
  n110 = TextureSampleParameter2D!(UVs=@n109, Apply View MipBias) {
   group=Custom01
   parameterName=Custom01_Diffuse
   samplerSource=SSM_Wrap_WorldGroupSettings
   texture=/Game/Environments/Helios/Roads/Textures/Helios_Brim_CobbleRoad_D
  }
  n111 = Saturate(None=Multiply(A=@n110.RGB, B=VectorParameter() {group=Custom01, parameterName=Custom01_Diffuse_ColorTint, defaultValue=<rgba 1 1 1 1>}.RGB))
  n112 = TextureSampleParameter2D(UVs=@n109) {
   group=Custom01
   parameterName=Custom01_SMR
   samplerSource=SSM_Wrap_WorldGroupSettings
   samplerType=SAMPLERTYPE_Masks
   texture=/Game/Environments/Helios/Roads/Textures/Helios_Brim_CobbleRoad_S
  }
  n113 = LinearInterpolate(
   A=ScalarParameter() {defaultValue=-0.05, group=Custom01, parameterName=Custom01_Noise_Rough_Min_Add}
   B=ScalarParameter() {defaultValue=0.4, group=Custom01, parameterName=Custom01_Noise_Rough_Max_Add}
   Alpha=@n104
  )
  n114 = TextureSampleParameter2D(UVs=@n109) {
   group=Custom01
   parameterName=Custom01_Normal
   samplerSource=SSM_Wrap_WorldGroupSettings
   samplerType=SAMPLERTYPE_Normal
   texture=/Game/Environments/Helios/Roads/Textures/Helios_Brim_CobbleRoad_N
  }
  n115 = Transform(None=@n114.RGB)
  n116 = MaterialFunctionCall(Normals=@n114.RGB, Normal Strength=ScalarParameter() {defaultValue=1, group=Custom01, parameterName=Custom01_NormalStrength}) {materialFunction=&MF_HQ_WorldAlignedNormal}
  n117 = SetMaterialAttributes!(
   MaterialAttributes
   Base Color=Saturate(None=Multiply(A=@n105, B=@n111))
   Metallic=Constant()
   Specular=@n112.R
   Roughness=Saturate(None=Add(A=@n113, B=@n112.B))
   Normal=LinearInterpolate(A=@n115, B=@n116.Result, Alpha=@n108)
   Opacity Mask=NamedRerouteUsage()
  ) {attributeSetTypes [
    #1
    #8
    #6
    #5
    #7
    #3
   ]}
  NamedRerouteDeclaration(None=@n117) {name=Custom01_Material, nodeColor=<rgba 0 0.4674 1 1>}
  NamedRerouteDeclaration(None=Multiply(A=NamedRerouteUsage() {desc=May need this version with the small snow detail breakup}, B=@n102.Alpha)) {name=Snow_ForPostRVT_Mask, nodeColor=<rgba 1 0.025 0.3669 1>}
  NamedRerouteDeclaration(None=@n103.Normal) {name=Snow_ForPostRVT_Normal, nodeColor=<rgba 1 0.025 0.3669 1>}
  n118 = ComponentMask(None=TextureCoordinate()) {g=true}
  n119 = MaterialFunctionCall!(
   InputMaterial=NamedRerouteUsage()
   LayerMaterial=NamedRerouteUsage()
   LayerWeight=LinearInterpolate(
    A=@n118
    B=OneMinus(None=@n118)
    Alpha=ScalarParameter() {defaultValue=1, group=BlendToCustom, parameterName=BlendToCustom_InvertBlendDirection}
   )
   UseContrast=StaticBool() {value=true}
   Contrast=ScalarParameter() {defaultValue=0.5, group=BlendToCustom, parameterName=BlendToCustom_Contrast}
  ) {materialFunction=/Game/Athena/Helios/Environments/Landscape/Materials/MaterialFunctions_Utilities/MF_HeightBlendLandscapeLayers_OpacityMask}
  n120 = NamedRerouteUsage()
  n121 = NamedRerouteUsage()
  NamedRerouteDeclaration(None=StaticSwitch(True=StaticSwitch(True=@n119.BlendedMaterial, False=@n120, Value=@n121), False=NamedRerouteUsage(), Value=NamedRerouteUsage())) {name=Prep_RoadMaterial, nodeColor=<rgba 0.5 0.5 0 1>}
  NamedRerouteDeclaration(None=@n94.SnowMaterialOnly) {name=SnowMaterialOnly_ForPostRVT, nodeColor=<rgba 1 0.025 0.3669 1>}
  n122 = NamedRerouteUsage()
  n123 = NamedRerouteUsage()
  n124 = StaticSwitch(
   True=StaticSwitch(True=LinearInterpolate(A=@n122, B=@n123, Alpha=@n119.LayerMask), False=NamedRerouteUsage(), Value=NamedRerouteUsage())
   False=NamedRerouteUsage()
   Value=NamedRerouteUsage()
  )
  NamedRerouteDeclaration(None=@n124) {name=Final_RoadOpacityMask, nodeColor=<rgba 0.5 0.5 0 1>}
  NamedRerouteDeclaration(None=LinearInterpolate(
    A=VectorParameter() {group=Road Lines, parameterName=EdgeLine, defaultValue=<rgba 0.7344 0.6848 0.6196 0>}.RGB
    B=VectorParameter() {group=Road Lines, parameterName=CenterLine, defaultValue=<rgba 0.4125 0.2346 0 1>}.RGB
    Alpha=NamedRerouteDeclaration(None=@n97.B) {name=RoadLineMask_B, nodeColor=<rgba 0.5469 0 1 1>}
   )) {name=LineColor, nodeColor=<rgba 0.5469 0 1 1>}
  NamedRerouteDeclaration(None=StaticBoolParameter() {group=Custom01, parameterName=Use_Custom01, sortPriority=-5000}) {name=Use_Custom01, nodeColor=<rgba 0 0.4674 1 1>}
  NamedRerouteDeclaration(None=@n91.GravelLayer) {name=LandscapeLayerMatch_Gravel, nodeColor=<rgba 0.5469 0 1 1>}
  n125 = NamedRerouteUsage()
  n126 = Constant() {r=1}
  n127 = NamedRerouteUsage()
  n128 = NamedRerouteDeclaration(None=Multiply(A=StaticSwitch(True=@n125, False=@n126, Value=@n127), B=NamedRerouteUsage())) {name=EdgeGradientOrFadeOut, nodeColor=<rgba 0.5469 0 1 1>}
  n129 = ScalarParameter() {defaultValue=1, group=Custom01, parameterName=Custom01_EdgeFalloff_Contrast}
  n130 = MaterialFunctionCall!(
   Diffuse 1
   Diffuse 2
   Transistion Phase=OneMinus(None=@n128)
   Height Texture 1=NamedRerouteUsage()
   Height Texture 2=NamedRerouteUsage()
   Contrast=@n129
  ) {materialFunction=/Engine/Functions/Engine_MaterialFunctions02/Texturing/HeightLerpWithTwoHeightMaps}
  n131 = MaterialFunctionCall(Transition Phase=@n128, Height Texture=NamedRerouteUsage(), Contrast=@n129) {materialFunction=&HeightLerp}
  n132 = NamedRerouteUsage()
  NamedRerouteDeclaration(None=ShadingPathSwitch(Default=StaticSwitch(True=@n130.Alpha, False=@n131.Alpha, Value=@n132), Mobile=@n131.Alpha)) {name=Custom01_OpacityMask, nodeColor=<rgba 0 0.4674 1 1>}
  NamedRerouteDeclaration(None=@n91.AsphaltLayer) {name=LandscapeLayerMatch_Asphalt, nodeColor=<rgba 0.5469 0 1 1>}
  NamedRerouteDeclaration(None=@n110.A) {name=Custom01_Height, nodeColor=<rgba 0.5469 0 1 1>}
  n133 = NamedRerouteUsage()
  n134 = NamedRerouteUsage()
  n135 = NamedRerouteUsage()
  NamedRerouteDeclaration(None=ShadingPathSwitch(
    Default=BlendMaterialAttributes(A=@n133, B=@n134, Alpha=@n135) {desc="Vertex blend type altered!", vertexAttributeBlendType=UseA}
    Mobile=NamedRerouteUsage()
   )) {name=AsphaltRoadMaterial_Base, nodeColor=<rgba 0.5469 0 1 1>}
  NamedRerouteDeclaration(None=@n91.SnowDynamic) {name=LandscapeLayerMatch_SnowDynamic, nodeColor=<rgba 0.5469 0 1 1>}
  NamedRerouteDeclaration(None=StaticBoolParameter() {group=Asphalt_SandEdge, parameterName=UseAsphalt_SandEdge, sortPriority=-9000}) {name=UseAsphalt_SandEdge, nodeColor=<rgba 0.5469 0 1 1>}
  n136 = TextureSample(UVs=Multiply(A=WorldPosition().XY, B=ScalarParameter() {defaultValue=0.0003, parameterName=Mobile_Asphalt_FadeOutNoiseScale}) {constB=0.0001}) {
   samplerSource=SSM_Wrap_WorldGroupSettings
   samplerType=SAMPLERTYPE_Masks
   texture=/Game/Environments/Helios/Terrain/TexturesMobile/T_Helios_Mobile_Noise
  }
  NamedRerouteDeclaration(None=@n136.B) {name=Mobile_Asphalt_FadeOutNoise, nodeColor=<rgba 0.5469 0 1 1>}
  n137 = Multiply(A=NamedRerouteUsage(), B=Constant() {r=0.85})
  NamedRerouteDeclaration(None=Multiply(A=OneMinus(None=@n137), B=NamedRerouteDeclaration(None=@n97.G) {name=RoadLineMask_G, nodeColor=<rgba 0.5469 0 1 1>})) {name=RoadLineMask, desc=HighEnd path only, nodeColor=<rgba 0.5469 0 1 1>}
  n138 = NamedRerouteUsage()
  NamedRerouteDeclaration(None=GetMaterialAttributes(None=@n138) {attributeGetTypes [#3]}.Opacity Mask) {name=LandscapeLayerMatch_Height, nodeColor=<rgba 0.5469 0 1 1>}
  n139 = StaticBoolParameter() {
   desc="Default=False. True will use the heightmap of a defined landscape layer for the falloff blend. Not used on mobile."
   group=Custom01
   parameterName=Use_FalloffToLandscapeLayerMatch
   sortPriority=-5000
  }
  NamedRerouteDeclaration(None=@n139) {name=Use_FalloffToLandscapeLayerMatch, nodeColor=<rgba 0.5469 0 1 1>}
  n140 = StaticBoolParameter() {
   desc="If this is set to True, then BlendAsphaltToCustom should be set to False"
   group=FadeOut
   parameterName=Use_FadeOut
   sortPriority=-5000
  }
  NamedRerouteDeclaration(None=@n140) {name=Use_FadeOut, nodeColor=<rgba 0.5469 0 1 1>}
  NamedRerouteDeclaration(None=@n68.Opacity Mask) {name=Mobile_AsphaltHeight, nodeColor=<rgba 0.5469 0 1 1>}
  NamedRerouteDeclaration(None=ShadingPathSwitch(Default=@n31, Mobile=NamedRerouteUsage())) {name=AsphaltSidesHeight_Modified, nodeColor=<rgba 0.5469 0 1 1>}
  n141 = StaticBoolParameter() {
   desc="If this is set to True, then Use_FadeOut should be set to False"
   group=BlendToCustom
   parameterName=BlendAsphaltToCustom
   sortPriority=-5000
  }
  NamedRerouteDeclaration(None=@n141) {name=BlendAsphaltToCustom, nodeColor=<rgba 0.5469 0 1 1>}
  n142 = MaterialFunctionCall!(
   Coordinates
   R_01_Scale=ScalarParameter() {defaultValue=8000, parameterName=NoiseScale_R01}
   R_02_Scale
   R_03_Scale
   R_04_Scale
   G_01_Scale
   G_02_Scale=Constant() {r=1650}
   G_03_Scale
   G_04_Scale
  ) {materialFunction=/Game/Athena/Helios/Environments/Landscape/Materials/MaterialFunctions_Utilities/MF_VariationMasks2}
  NamedRerouteDeclaration(None=@n142.G_02) {name=SharedNoise_G_02, nodeColor=<rgba 0.5469 0 1 1>}
  NamedRerouteDeclaration(None=Max(A=@n142.G_01, B=@n142.G_01)) {name=SharedNoise_G01MAXG04, nodeColor=<rgba 0.5469 0 1 1>}
  NamedRerouteDeclaration(None=@n34.Normal) {name=High_AsphaltNormal, nodeColor=<rgba 0.5469 0 1 1>}
  NamedRerouteDeclaration(None=GetMaterialAttributes(None=NamedRerouteDeclaration(None=@n91.SandLayer) {name=LandscapeLayerMatch_Sand, nodeColor=<rgba 0.5469 0 1 1>}) {attributeGetTypes [#3]}.Opacity Mask) {name=SandHeight, nodeColor=<rgba 0.5469 0 1 1>}
  n143 = VectorParameter() {
   group=Asphalt_SandEdge
   parameterName=AsphaltSand_RoadType_EdgeFalloff
   sortPriority=-200
   channelNames {a=Add, b=Multiply, g=FalloffMax, r=FalloffMin}
   defaultValue=<rgba -0.05 0.1 1 0>
  }
  n144 = Multiply(A=SmoothStep(Min=@n143.FalloffMin, Max=@n143.FalloffMin, Value=NamedRerouteUsage()) {constMax=0.5}, B=@n143.Multiply)
  n145 = Multiply(A=StaticSwitch(True=NamedRerouteUsage(), False=Constant() {r=1}, Value=NamedRerouteUsage()), B=Saturate(None=Add(A=@n144, B=@n143.Add)))
  n146 = MaterialFunctionCall(Transition Phase=@n145, Height Texture=NamedRerouteUsage()) {materialFunction=&HeightLerp}
  n147 = ScalarParameter() {defaultValue=5, group=Asphalt_SandEdge, parameterName=AsphaltSand_EdgeFalloff_Mod}
  n148 = Saturate(None=Add(A=Multiply(A=@n146.Lerp Alpha No Contrast, B=@n147), B=ScalarParameter() {group=Asphalt_SandEdge, parameterName=AsphaltSand_EdgeFalloff_Add}))
  n149 = Multiply(A=Add(A=NamedRerouteUsage(), B=@n145), B=@n145)
  n150 = ScalarParameter() {defaultValue=5, group=Asphalt_SandEdge, parameterName=AsphaltSand_EdgeFalloff_Mod_Mobile}
  n151 = Saturate(None=Add(
    A=Multiply(A=@n149, B=@n150) {constB=5}
    B=ScalarParameter() {defaultValue=-1, group=Asphalt_SandEdge, parameterName=AsphaltSand_EdgeFalloff_Add_Mobile}
   ) {constB=0})
  NamedRerouteDeclaration(None=ShadingPathSwitch(Default=@n148, Mobile=QualitySwitch(Default=@n148, Medium=@n151))) {name=SandOuterEdgeHeight, nodeColor=<rgba 0.5469 0 1 1>}
  NamedRerouteDeclaration(None=@n142.R_01) {name=Noise_R01, nodeColor=<rgba 0.5469 0 1 1>}
  NamedRerouteDeclaration(None=@n37.G) {name=AsphaltDiffuse_G, nodeColor=<rgba 0.5469 0 1 1>}
  n152 = WorldPosition()
  n153 = ScalarParameter() {defaultValue=8000, parameterName=NoiseScale_B01}
  NamedRerouteDeclaration(None=TextureSample(UVs=Divide(A=@n152.XY, B=@n153)) {samplerSource=SSM_Wrap_WorldGroupSettings, texture=/Game/Environments/Helios/Terrain/Textures/T_Helios_Landscape_Noise}.B) {name=Noise_B01, nodeColor=<rgba 0.5469 0 1 1>}
  NamedRerouteDeclaration(None=@n61.RGB) {name=RoadLineNormalTexture, nodeColor=<rgba 0.5469 0 1 1>}
  n154 = SetMaterialAttributes!(
   MaterialAttributes=@n34.MaterialAttributes
   Roughness=LinearInterpolate(A=@n34.Roughness, B=NamedRerouteUsage(), Alpha=@n2)
   Specular=LinearInterpolate(A=@n34.Specular, B=Constant() {r=0.6}, Alpha=@n2) {constB=0.6}
   Base Color=LinearInterpolate(A=@n34.Base Color, B=NamedRerouteUsage(), Alpha=@n2)
   Normal=@n3
   Opacity Mask=NamedRerouteUsage()
  ) {attributeSetTypes [
    #5
    #6
    #1
    #7
    #3
   ]}
  NamedRerouteDeclaration(None=@n154) {name=AsphaltWithRoadLines, nodeColor=<rgba 0.5469 0 1 1>}
  n155 = ScalarParameter() {group=Asphalt_SandEdge, parameterName=AsphaltSand_AsphaltHeightLow}
  n156 = ScalarParameter() {defaultValue=1, group=Asphalt_SandEdge, parameterName=AsphaltSand_AsphaltHeightHigh}
  n157 = Multiply(A=NamedRerouteUsage(), B=ScalarParameter() {defaultValue=10, group=Asphalt_SandEdge, parameterName=AsphaltSand_Asphalt_Height_Mod})
  n158 = ScalarParameter() {group=Asphalt_SandEdge, parameterName=AsphaltSand_AsphaltHeight_Add}
  n159 = OneMinus(None=Saturate(None=Add(A=@n157, B=@n158)))
  n160 = NamedRerouteUsage()
  n161 = Constant() {r=1}
  n162 = NamedRerouteUsage()
  n163 = OneMinus(None=Multiply(A=StaticSwitch(True=@n160, False=@n161, Value=@n162), B=NamedRerouteUsage()))
  n164 = Power(Base=@n163, Exp=ScalarParameter() {defaultValue=25, group=Asphalt_SandEdge, parameterName=AsphaltSand_InteriorFalloff_PowerShort})
  n165 = Multiply(
   A=@n163
   B=Power(
    Base=NamedRerouteUsage()
    Exp=ScalarParameter() {defaultValue=0.5, group=Asphalt_SandEdge, parameterName=AsphaltSand_InteriorFalloff_NoisePower}
   )
  )
  n166 = MaterialFunctionCall(
   InputMaterial=SetMaterialAttributes!(MaterialAttributes=NamedRerouteUsage(), Opacity Mask=LinearInterpolate(A=@n155, B=@n156, Alpha=@n159)) {attributeSetTypes [#3]}
   LayerMaterial=NamedRerouteUsage()
   LayerWeight=Multiply(A=Max(A=@n164, B=@n165), B=@n54)
   UseContrast=StaticBool()
  ) {materialFunction=&MF_HeightBlendLandscapeLayers_OpacityMask}
  NamedRerouteDeclaration(None=StaticSwitch(True=@n166.BlendedMaterial, False=NamedRerouteUsage(), Value=NamedRerouteUsage())) {name=AsphaltRoadMaterial, nodeColor=<rgba 0.5469 0 1 1>}
  NamedRerouteUsage()
  NamedRerouteUsage()
  NamedRerouteUsage()
  NamedRerouteUsage()
  NamedRerouteUsage()
  NamedRerouteUsage()
  OneMinus(None=NamedRerouteUsage())
  Power(Base=@n163, Exp=ScalarParameter() {defaultValue=1, group=Asphalt_SandEdge, parameterName=AsphaltSand_InteriorFalloff_PowerLong})
  n167 = StaticBoolParameter() {
   defaultValue=true
   desc="False = Use LandscapeLayerMatch_Sand. Default=True"
   group=Custom01
   parameterName=Use_LandscapeLayerMatch_Gravel
   sortPriority=-5000
  }
  Reroute(None=StaticSwitch(True=@n138, False=NamedRerouteUsage(), Value=@n167))
  n168 = GetMaterialAttributes(None=NamedRerouteUsage()) {attributeGetTypes [
    #1
    #6
    #5
    #7
   ]}
  RuntimeVirtualTextureOutput!(
   BaseColor=@n168.Base Color
   Specular=@n168.Base Color
   Roughness=@n168.Base Color
   Normal=@n168.Base Color
   WorldHeight
   Opacity=NamedRerouteUsage()
   Mask
   Displacement
   Mask4
  )
  Saturate(None=Add(A=@n112.B, B=ScalarParameter() {group=Custom01, parameterName=Custom01_Roughness_Add}))
  Saturate(None=Add(A=@n124, B=ScalarParameter() {group=RoadSnow, parameterName=DEBUG_AddToOpacity}))
  ScalarParameter() {defaultValue=1, group=RoadSnow, parameterName=RoadSnow_Noise_Max}
  ScalarParameter() {defaultValue=-0.2, group=RoadSnow, parameterName=RoadSnow_Noise_Min}
  ScalarParameter() {defaultValue=4096, group=RoadSnow, parameterName=RoadSnow_Noise_Scale}
  ScalarParameter() {group=RoadSnow, parameterName=RoadSnow_RoadEdgeHardness}
  SetMaterialAttributes!(MaterialAttributes=@n33.MaterialAttributes, Base Color=LinearInterpolate(A=NamedRerouteUsage(), B=@n33.Base Color, Alpha=@n32.ColorAmount)) {attributeSetTypes [#1]}
  n169 = TextureCoordinate()
  n170 = Frac(None=MaterialFunctionCall(X=ComponentMask(None=@n169) {r=true}, Y=ComponentMask(None=@n169) {g=true}) {materialFunction=&MakeFloat3}.Result)
  n171 = ComponentMask(None=VectorParameter() {group=Debug, parameterName=Override Color, defaultValue=<rgba 1 0 0 1>}.RGB) {b=true, g=true, r=true}
  StaticSwitchParameter(
   True=SetMaterialAttributes(Base Color=StaticSwitchParameter(True=@n170, False=@n171) {defaultValue=true, desc=Use this to show the spline mesh UV directionality, group=Debug, parameterName=Show UV's}) {attributeSetTypes [#1]}
   False=Reroute()
  ) {
   desc="Use this to identify where materials are used in the map\r\nand also the spline mesh UVs. Dont submit enabled!"
   group=Debug
   parameterName=Enable Debug Color
  }
  StaticSwitch(
   True=@n107.Result
   False=@n106
   Value=NamedRerouteDeclaration(None=StaticBoolParameter() {defaultValue=true, desc=False uses UV0, group=Custom01, parameterName=Custom01_UseWorlCoordinates}) {name=Custom01_UseWorlCoordinates, nodeColor=<rgba 0.5469 0 1 1>}
  )
  StaticSwitch(True=@n116.Result, False=@n115, Value=NamedRerouteUsage())
 ]
}
